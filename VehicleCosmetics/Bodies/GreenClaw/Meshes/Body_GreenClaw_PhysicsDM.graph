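# Physics asset for the GreenClaw vehicle cosmetic body mesh, providing its simplified collision shape.
# It has a single body on the root bone, made of one box offset slightly along X and raised to enclose the vehicle body as one rigid volume.
# No other bones have bodies and there are no constraints, so the asset defines no articulated or ragdoll parts.
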PhysicsAsset /VehicleCosmetics/Bodies/GreenClaw/Meshes/Body_GreenClaw_PhysicsDM
bodies {root [{shapeType=Box, extentX=444, extentY=216, extentZ=132, shapeName=root_box, center=<-7 0 108>}]}
constraints []
