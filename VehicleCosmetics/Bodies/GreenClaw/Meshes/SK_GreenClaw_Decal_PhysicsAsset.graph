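# Physics asset for SK_GreenClaw_Decal, a skeletal mesh in the GreenClaw vehicle body cosmetic set.
# It holds a single body on the Body bone, shaped as one capsule that is offset and rotated from the bone origin.
# It has no constraints, so there is no articulation between parts.
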PhysicsAsset /VehicleCosmetics/Bodies/GreenClaw/Meshes/SK_GreenClaw_Decal_PhysicsAsset
bodies {Body [{shapeType=Capsule, length=100.2056, radius=24.7608, center=<-62.7278 -35.5688 -10.7763>, rotation=<rot 0 -90 -9.7479>}]}
constraints []
